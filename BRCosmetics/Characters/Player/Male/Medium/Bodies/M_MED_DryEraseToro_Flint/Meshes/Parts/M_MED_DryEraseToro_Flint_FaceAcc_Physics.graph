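# Physics asset for the face accessory mesh of the male medium DryEraseToro Flint character cosmetic, defining physics for a dangling necklace.
# Two short capsule bodies form a chain from the thin dyn_necklace_MSTR root capsule, linked by constraints with moderate swing and very tight twist limits.
# A large, unconstrained spine_05 capsule stands in for the upper torso, likely as a collider for the necklace.
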PhysicsAsset /BRCosmetics/Characters/Player/Male/Medium/Bodies/M_MED_DryEraseToro_Flint/Meshes/Parts/M_MED_DryEraseToro_Flint_FaceAcc_Physics
bodies {
 dyn_necklace_01 [{shapeType=Capsule, length=4, radius=2, shapeName=dyn_necklace_01_capsule, center=<2 0 2>, rotation=<rot 90 0 0>}]
 dyn_necklace_02 [{shapeType=Capsule, length=4, radius=2, shapeName=dyn_necklace_02_capsule, center=<2 0 2>, rotation=<rot 90 0 0>}]
 dyn_necklace_MSTR [{shapeType=Capsule, length=12, radius=0.1, shapeName=dyn_necklace_MSTR_capsule}]
 spine_05 [{shapeType=Capsule, length=14.0459, radius=10.7532, shapeName=spine_05_capsule, center=<-0.3602 -0.1202 0>}]
}
constraints [
 {bone1Name=dyn_necklace_01, bone2Name=dyn_necklace_MSTR, swing1LimitDegrees=40, swing2LimitDegrees=20, twistLimitDegrees=2}
 {bone1Name=dyn_necklace_02, bone2Name=dyn_necklace_01, swing1LimitDegrees=40, swing2LimitDegrees=30, twistLimitDegrees=2}
]
