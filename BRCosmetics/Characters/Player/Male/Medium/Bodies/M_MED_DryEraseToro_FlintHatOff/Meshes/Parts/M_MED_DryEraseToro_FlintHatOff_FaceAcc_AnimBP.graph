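# Post-physics AnimBlueprint for the DryEraseToro FlintHatOff character's FaceAcc part, adding secondary motion, including necklace bones.
# It copies the attached parent's pose, then bRigidBodyEnabled selects RigidBody simulation or two chained spring-driven AnimDynamics necklace bones.
# RigidBody reuses the Flint variant's physics asset in spine_05 space.
# Each update stores parameter-asset gravity override settings and sets AnimDynamics space and alpha variables by front-end state.
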AnimBlueprint /BRCosmetics/Characters/Player/Male/Medium/Bodies/M_MED_DryEraseToro_FlintHatOff/Meshes/Parts/M_MED_DryEraseToro_FlintHatOff_FaceAcc_AnimBP
target_skeleton=&/M_MED_DryEraseToro_FlintHatOff_FaceAcc_Skeleton
graphs [
 AnimGraph {nodes [
   n1 = RigidBody!(
    ComponentPose=Local To Component(LocalPose=Copy Pose From Mesh() {bUseAttachedParent=true})
    Alpha
    ExternalForce
    SimSpaceSettings:Sim Space Settings Structure=(WorldAlpha=1,VelocityScaleZ=0.5,DampingAlpha=0,MaxLinearVelocity=800,MaxAngularVelocity=12,MaxLinearAcceleration=4000,MaxAngularAcceleration=60,ExternalLinearDragV=(X=0,Y=0,Z=0),ExternalLinearVelocity=(X=0,Y=0,Z=0),ExternalAngularVelocity=(X=0,Y=0,Z=0))
   ) -> (Pose) {
    bOverrideWorldGravity=true
    overridePhysicsAsset=/BRCosmetics/Characters/Player/Male/Medium/Bodies/M_MED_DryEraseToro_Flint/Meshes/Parts/M_MED_DryEraseToro_Flint_FaceAcc_Physics
    simulationSpace=BaseBoneSpace
    alphaBoolBlend {blendOption=Linear}
    baseBoneRef {boneName=spine_05}
    overrideWorldGravity {z=-980}
    simSpaceSettings {
     dampingAlpha=0
     maxAngularAcceleration=60
     maxAngularVelocity=12
     maxLinearAcceleration=4000
     maxLinearVelocity=800
     velocityScaleZ=0.5
     worldAlpha=1
    }
   }
   n2 = Anim Dynamics - Bone: dyn_necklace_01!(
    SimulationSpace:AnimPhysSimSpaceType Enum=Component
    ComponentPose=Local To Component(LocalPose=Copy Pose From Mesh() {bUseAttachedParent=true})
    Alpha=1
    ExternalForce="0,0,0"
   ) -> (Pose) {
    angularDampingOverride=0.99
    angularSpringConstant=10000
    bAngularSpring=true
    bOverrideAngularDamping=true
    bUseGravityOverride=true
    bUsePlanarLimit=false
    gravityScale=0.35
    alphaBoolBlend {blendOption=Linear}
    boundBone {boneName=dyn_necklace_01}
    componentAppliedLinearAccClamp=<2500 2500 2500>
    componentLinearAccScale=<1 1 1>
    componentLinearVelScale=<1 1 1>
    gravityOverride {z=-980}
    physicsBodyDefinitions [{
      boundBone {boneName=dyn_necklace_01}
      boxExtents=<20 30 30>
      constraintSetup {angularLimitsMax=<0 0 10>, angularLimitsMin=<0 -15 -10>, angularTarget=<1 0 0>}
      localJointOffset=<4 0 0>
     }]
    planarLimits [{drivingBone {boneName=spine_05}, planeTransform {location=<-0.1567 10.8941 0>, rotation=<rot 0 -17.2798 90.0001>}}]
    relativeSpaceBone {boneName=spine_05}
   }
   n3 = Anim Dynamics - Bone: dyn_necklace_02!(SimulationSpace:AnimPhysSimSpaceType Enum=Component, ComponentPose=@n2, Alpha=1, ExternalForce="0,0,0") -> (Pose) {
    angularDampingOverride=0.99
    angularSpringConstant=10000
    bAngularSpring=true
    bOverrideAngularDamping=true
    bUseGravityOverride=true
    bUsePlanarLimit=false
    gravityScale=0.35
    alphaBoolBlend {blendOption=Linear}
    boundBone {boneName=dyn_necklace_02}
    componentAppliedLinearAccClamp=<2500 2500 2500>
    componentLinearAccScale=<1 1 1>
    componentLinearVelScale=<1 1 1>
    gravityOverride {z=-980}
    physicsBodyDefinitions [{
      boundBone {boneName=dyn_necklace_02}
      boxExtents=<20 30 30>
      constraintSetup {angularLimitsMax=<0 0 10>, angularLimitsMin=<0 -15 -10>, angularTarget=<1 0 0>}
      localJointOffset=<4 0 0>
     }]
    planarLimits [{drivingBone {boneName=spine_05}, planeTransform {location=<-2.2273 11.5003 -3.3986>, rotation=<rot 0 -19.4217 90>}}]
    relativeSpaceBone {boneName=spine_05}
   }
   Output Pose(Result=Blend Poses by bool(BlendPose_0=Component To Local(ComponentPose=@n1), BlendPose_1=Component To Local(ComponentPose=@n3), bActiveValue=Get bRigidBodyEnabled()) {blendType=Linear}) {layerGroup=None}
  ]}
 EventGraph {nodes [
   Event BlueprintUpdateAnimation()
   > n1 = Sequence()
   > n2 = "Gravity Override State Function (Parameter Asset) (Nativized)"(
    CharacterPartAnimInstance=Self-Reference()
    GameTimeInSeconds=GetGameTimeInSeconds()
    Parameters=/BRCosmetics/Characters/Player/Male/Medium/Bodies/M_MED_DryEraseToro_Flint/Meshes/Parts/M_MED_DryEraseToro_Flint_FA_Parameters
   )
   > Set RBAN Settings Out(RBAN Settings Out=@n2.OutRigidBodyAnimNodeParameters)
   n3 = Branch(execute=@n1.then_1, Condition=Get bIsInFrontEnd())
   > Set Anim Dynamics Sim Space(Anim Dynamics Sim Space:AnimPhysSimSpaceType Enum=World) -> (Output_Get:AnimPhysSimSpaceType Enum)
   > Set Anim Dynamics Alpha(Anim Dynamics Alpha:double=0.4) -> (Output_Get:double)
   Set Anim Dynamics Sim Space(execute=@n3.else)
   > Set Anim Dynamics Alpha(Anim Dynamics Alpha=1)
  ]}
]
variables [
 Anim Dynamics Alpha {type=number}
 Anim Dynamics Sim Space {type=AnimPhysSimSpaceType}
 RBAN Settings Out {type=FortRigidBodyAnimNodeParameters}
]
